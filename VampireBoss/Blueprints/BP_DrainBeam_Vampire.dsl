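; Vampire boss drain beam actor that visually links the Monster to the drained Player; its logic is cosmetic and runs only on non-authority machines.
; Each Tick feeds both characters' spine_05 socket locations into the looping LifeDrain beam effect as spline start and end, skipping a zero end position.
; With a valid Player, a DoOnce waits briefly, then, if CanAddBeam, adds a local looping Drained GameplayCue sourced from the Monster.
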(event EventBeginPlay)

(event Collision|EventActorBeginOverlap
  (params (OtherActor :type Actor%20Object%20Reference))
  (|CallParentFunction OtherActor) ; This node is disabled and will not be called.\nDrag off pins to build functionality.
)

(event EventTick
  (params (DeltaSeconds :type Float%20%28single-precision%29))
  (bind _ns_lifedrain_beam_loop (Variables|Default|GetNS_LifeDrain_Beam_Loop))
  (bind _monster (Variables|InternalData|GetMonster))
  (Networking|SwitchHasAuthority
    (:Remote
      (Effects|Components|ParticleSystem|SetVectorParameter _ns_lifedrain_beam_loop "SplineUpdatedStartPos" (Transformation|GetSocketLocation (Variables|/Script/Engine.Character|GetMesh (Utilities|Casting|CastToFortPlayerPawn _monster)) "spine_05"))
      (bind _player (Variables|InternalData|GetPlayer)
        (:then
          (Utilities|FlowControl|Sequence
            (:then_0
              (Networking|SwitchHasAuthority
                (:Remote
                  (Variables|InternalData|SetDrained%20Cue%20Parameters (Ability|GameplayCue|MakeGameplayCueParameters :NormalizedMagnitude 0.0 :RawMagnitude 0.0 :Location "0,0,0" :Normal "0,0,0" :SourceObject _monster :GameplayEffectLevel 1 :AbilityLevel 1 :bReplicateLocationWhenUsingMinimalRepProxy false))
                  (Utilities|FlowControl|DoOnce false
                    (:Completed
                      (Utilities|FlowControl|Delay 0.35)
                      (if (Variables|Default|GetCanAddBeam)
                        (Projectile|AddGameplayCue%28Looping%29Local (Variables|InternalData|GetPlayer) (Variables|Default|GetDrainedGameplayCueTag) (Variables|InternalData|GetDrained%20Cue%20Parameters) false)))))))
            (:then_1
              (if (Math|Vector|NotEqual(Vector) (Transformation|GetSocketLocation (Variables|/Script/Engine.Character|GetMesh _player) "spine_05") "0, 0, 0" 0.0001)
                (Effects|Components|ParticleSystem|SetVectorParameter _ns_lifedrain_beam_loop "SplineUpdatedEndPos" (Transformation|GetSocketLocation (Variables|/Script/Engine.Character|GetMesh _player) "spine_05"))))))
        (:else)))))

(event Custom|CustomEvent)

(fn ConstructionScript
  (|CallParentFunction))
