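; Blueprint that, judging by its name, is a looping life-drain camera lens effect for the Vampire boss.
; Its Tick, BeginPlay and Destroyed events and its construction script are empty, so it adds no graph logic of its own.
; The parent class and any effect or component settings do not appear in the graphs.
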(event EventTick
  (params (DeltaSeconds :type Float%20%28single-precision%29)))

(event EventBeginPlay)

(event EventDestroyed)

(fn ConstructionScript)
