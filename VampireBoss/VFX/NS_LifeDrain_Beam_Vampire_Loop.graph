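# Looping Niagara life-drain beam for the vampire boss, drawn between user-fed start and end positions.
# SplineElectric bends a ribbon along a Hermite spline with helical wobble, thinning and flickering as the endpoints separate and dissolving on RemoveBeam.
# Mesh bats swirl at the end point, then get pulled toward the start.
# Glow travels start to end emitting location events, alongside flame sprites; purple ribbon and BirdBat001 variants are disabled.
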NiagaraSystem /VampireBoss/VFX/NS_LifeDrain_Beam_Vampire_Loop
dependencies {
 usedDataInterfaces [&NiagaraDataInterfaceCurve, &NiagaraDataInterfaceColorCurve]
 usedDynamicInputs [&Abs_Float, &Subtract_Position]
 usedModules [
  &SystemState
  &self:SystemSpawnScript.SystemScriptSource.SystemScriptGraph.NiagaraNodeAssignment_0.SetVariables_#1
  &Timeline
 ]
}
emitters {
 Bats {
  emitterUpdateScript [
   &EmitterState(Life Cycle Mode=::NewEnumerator1, Loop Behavior=::NewEnumerator1, Loop Duration=2, UseLoopDelay=false)
   &SpawnRate(
    SpawnRate=&MakeCustomFloatFromBool(Bool=$System.Check Distance, True Float=0, False Float=3)
    Use Spawn Probability=false
   )
  ]
  particleSpawnScript [
   &InitializeParticle(
    Lifetime Max=3
    Lifetime Min=3
    Lifetime Mode=::NewEnumerator1
    Mass Mode=::NewEnumerator3
    Mesh Scale Mode=::NewEnumerator1
    Mesh Uniform Scale Max=0.8
    Mesh Uniform Scale Min=0.5
    Sprite Rotation Mode=::NewEnumerator2
    Sprite Size Mode=::NewEnumerator3
    Uniform Sprite Size=150
    UsePositionOffset=false
    Write Material Random=true
   )
   &ShapeLocation(
    Cylinder Distribution=::NewEnumerator2
    Cylinder Height=120
    MaskSpawnGroup=false
    Shape Origin=$User.SplineUpdatedEndPos
    Shape Primitive=::NewEnumerator1
    U Position=&RandomRangeFloat(Minimum=-0.1, Maximum=0.1)
    UseOwnerScale=true
    V Position=&RandomRangeFloat(Minimum=0, Maximum=1)
   )
   /SetVariables(Particles.RandSign:NiagaraFloat=&UniformAOrBFloat(A=-1)) {isSetParameters=true, module=SetVariables_#2}
  ]
  particleUpdateScript [
   &ParticleState()
   /SetVariables(Particles.Scale2:Vector3f=&Multiply_VectorByFloat(
     Vector=$Particles.Initial.Scale
     Float=&FloatFromCurve(
      FloatCurve=DataInterface() {propertyValues {Curve {keys [
          {interpMode=RCIM_Cubic, leaveTangentWeight=0.0333, tangentMode=RCTM_User, time=0.7, value=1}
          {arriveTangentWeight=0.17, interpMode=RCIM_Cubic, leaveTangentWeight=0.17, tangentMode=RCTM_User, time=0.8, value=1}
          <at 0.8476 1.007>
          {arriveTangentWeight=0.2247, interpMode=RCIM_Cubic, leaveTangentWeight=0.1124, tangentMode=RCTM_User, time=0.95, value=1}
          {arriveTangentWeight=0.0167, interpMode=RCIM_Cubic, tangentMode=RCTM_User, time=1, value=0}
         ]}}}
      Scale Curve=&FloatFromCurve(FloatCurve=DataInterface() {propertyValues {Curve {keys [<at 0 1>,<at 1 0>]}}}, CurveIndex=$System.Timeline.NormalizedTime)
     )
    )) {isSetParameters=true, module=SetVariables_#3}
   &UpdateMeshOrientation(Bank Rate=90, Banking Angle Min / Max:Vector2f=<-5 5>, Orientation Method=::NewEnumerator3)
   &ScaleColor(
    Scale RGB=&VectorFromFloat(Value=&FloatFromCurve(FloatCurve=DataInterface() {propertyValues {Curve {keys [<at 0.75 0>,<at 0.85 0>]}}}))
    ScaleA=true
    ScaleRGB=true
   )
   /Niagara/Modules/Update/Velocity/VortexVelocity(
    Module.Use Influence Falloff=false
    Velocity Amount:NiagaraFloat=&FloatFromCurve(
     FloatCurve=DataInterface() {propertyValues {Curve {keys [<at 0.5 1>,<at 0.8 0>]}}}
     Scale Curve=&Multiply_Float(A=20000, B=$Particles.RandSign)
    )
    Vortex Origin:NiagaraPosition=$User.SplineUpdatedEndPos
   )
   &CurlNoiseForce(
    Noise Frequency=&FloatFromCurve(FloatCurve=DataInterface() {propertyValues {Curve {keys [<at 0.5 120>,<at 0.8 10>]}}})
    Noise Strength=&FloatFromCurve(FloatCurve=DataInterface() {propertyValues {Curve {keys [<at 0.5 0.5>,<at 0.8 1>]}}}, Scale Curve=10000)
    Pan Noise=false
   )
   &PointAttractionForce(
    Attraction Radius=1000000000
    Attraction Strength=&FloatFromCurve(
     FloatCurve=DataInterface() {propertyValues {Curve {keys [{tangentMode=RCTM_User, time=0.5, value=0}, {tangentMode=RCTM_User, time=0.8, value=1}]}}}
     Scale Curve=1000
    )
    Attractor Position:NiagaraPosition=$User.SplineUpdatedStartPos
    Kill Within Radius=false
    Module.Use Falloff=true
    Position To Attract:NiagaraPosition=&AddVectorToPosition(Position=$Particles.Position, Vector=<0 0 0>)
    UseAttractorOffset=false
   ) {module=PointAttractionForce001}
   &SolveForcesAndVelocity(Clamp Velocity=false, Limit Acceleration=false)
   &DynamicMaterialParameters(
    Index 0 Param 1=&RandomRangeFloat(Minimum=0, Maximum=10)
    Index 0 Param 2=&RandomRangeFloat(Minimum=3, Maximum=3.5)
    Index 0 Param 3=&FloatFromCurve(
     FloatCurve=DataInterface() {propertyValues {Curve {keys [
         {interpMode=RCIM_Cubic, tangentMode=RCTM_User, time=0, value=0}
         {interpMode=RCIM_Cubic, tangentMode=RCTM_User, time=0.1, value=1}
         {arriveTangentWeight=0.3202, interpMode=RCIM_Cubic, leaveTangentWeight=0.1067, tangentMode=RCTM_User, time=0.75, value=1}
         {arriveTangentWeight=0.0667, interpMode=RCIM_Cubic, tangentMode=RCTM_User, time=0.85, value=0}
        ]}}}
     Scale Curve=0.725
    )
    Index 1 Param 1=&RandomRangeFloat(Minimum=0, Maximum=3)
    Index 1 Param 2=&RandomRangeFloat(Minimum=0.9, Maximum=1.25)
    Module.Index0Param0Write=true
    Module.Index0Param1Write=true
    Module.Index0Param2Write=true
    Module.Index0Param3Write=true
    Module.Index1Param0Write=true
    Module.Index1Param1Write=true
    Module.Index1Param2Write=true
    Module.Index1Param3Write=true
    Write Parameter Index 1=true
   )
   &KillParticlesInVolume(Origin Offset:NiagaraPosition=$User.SplineUpdatedEndPos, Sphere Radius=10)
   /SetVariables(Particles.LerpPosition:NiagaraPosition=&LerpPosition(
     A=$Particles.Position
     B=$User.SplineUpdatedStartPos
     Alpha=&FloatFromCurve(FloatCurve=DataInterface() {propertyValues {Curve {keys [
          {
           arriveTangentWeight=0.0497
           interpMode=RCIM_Cubic
           leaveTangentWeight=0.0497
           tangentMode=RCTM_User
           tangentWeightMode=RCTWM_WeightedBoth
           time=0.85
           value=0
          }
          {
           arriveTangentWeight=0.0558
           interpMode=RCIM_Cubic
           leaveTangentWeight=0.0558
           tangentMode=RCTM_User
           tangentWeightMode=RCTWM_WeightedBoth
           time=1
           value=1
          }
         ]}}})
    )) {isSetParameters=true, module=SetVariables_#4}
  ]
  rendererClasses [&NiagaraMeshRendererProperties]
  renderers [{rendererClass=&NiagaraMeshRendererProperties, rendererIndex=0}]
 }
 BirdBat001 {
  bEnabled=false
  emitterUpdateScript [
   &EmitterState(Life Cycle Mode=::NewEnumerator1, Loop Behavior=::NewEnumerator1, Loop Duration=2, UseLoopDelay=false)
   &SpawnRate(
    SpawnRate=&MakeCustomFloatFromBool(Bool=$System.Check Distance, True Float=0, False Float=3)
    Use Spawn Probability=false
   )
  ]
  particleSpawnScript [
   &InitializeParticle(
    Lifetime Max=3
    Lifetime Min=3
    Lifetime Mode=::NewEnumerator1
    Mass Mode=::NewEnumerator3
    Mesh Scale Mode=::NewEnumerator1
    Mesh Uniform Scale Max=1
    Sprite Rotation Mode=::NewEnumerator2
    Sprite Size Mode=::NewEnumerator3
    Uniform Sprite Size=150
    UsePositionOffset=false
    Write Material Random=true
   )
   &ShapeLocation(
    Cylinder Distribution=::NewEnumerator2
    Cylinder Height=120
    MaskSpawnGroup=false
    Shape Origin=$User.SplineUpdatedEndPos
    Shape Primitive=::NewEnumerator1
    U Position=&RandomRangeFloat(Minimum=-0.1, Maximum=0.1)
    UseOwnerScale=true
    V Position=&RandomRangeFloat(Minimum=0, Maximum=1)
   )
   /SetVariables(Particles.RandSign:NiagaraFloat=&UniformAOrBFloat(A=-1)) {isSetParameters=true, module=SetVariables_#2}
  ]
  particleUpdateScript [
   &ParticleState()
   &ScaleSpriteSize(
    Uniform Curve Index=$System.Timeline.NormalizedTime
    Uniform Curve Scale=2
    Uniform Curve Sprite Scale=DataInterface() {propertyValues {Curve {keys [{tangentMode=RCTM_User, time=0, value=1}, {tangentMode=RCTM_User, time=1, value=0}]}}}
   )
   &ScaleSpriteSize(
    Uniform Curve Scale=2
    Uniform Curve Sprite Scale=DataInterface() {propertyValues {Curve {keys [
        {arriveTangent=57.0191, arriveTangentWeight=0.2966, interpMode=RCIM_Cubic, leaveTangent=57.0191, tangentMode=RCTM_User, time=0.7899, value=0}
        {arriveTangentWeight=0.0079, interpMode=RCIM_Cubic, leaveTangent=-34.6908, tangentMode=RCTM_Break, time=0.8136, value=1}
        {arriveTangentWeight=0.0069, interpMode=RCIM_Cubic, tangentMode=RCTM_User, time=0.8342, value=0}
       ]}}}
   ) {module=ScaleSpriteSize001}
   &ScaleMeshSize(Scale Factor=&VectorFromFloat(Value=&FloatFromCurve(FloatCurve=DataInterface() {propertyValues {Curve {keys [
          {interpMode=RCIM_Cubic, tangentMode=RCTM_User, time=0, value=0}
          {interpMode=RCIM_Cubic, tangentMode=RCTM_User, time=0.1, value=1}
          {arriveTangentWeight=0.3202, interpMode=RCIM_Cubic, leaveTangentWeight=0.1067, tangentMode=RCTM_User, time=0.75, value=1}
          {arriveTangentWeight=0.0667, interpMode=RCIM_Cubic, tangentMode=RCTM_User, time=0.85, value=0}
         ]}}})))
   &ScaleMeshSize(Scale Factor=&VectorFromFloat(Value=&FloatFromCurve(
      FloatCurve=DataInterface() {propertyValues {Curve {keys [{tangentMode=RCTM_User, time=0, value=1}, {tangentMode=RCTM_User, time=1, value=0}]}}}
      CurveIndex=$System.Timeline.NormalizedTime
     ))) {module=ScaleMeshSize002}
   &ScaleMeshSize(Scale Factor=&VectorFromFloat(Value=&FloatFromCurve(FloatCurve=DataInterface() {propertyValues {Curve {keys [
          {interpMode=RCIM_Cubic, tangentMode=RCTM_User, time=0, value=0}
          {interpMode=RCIM_Cubic, tangentMode=RCTM_User, time=0.1, value=1}
          {arriveTangentWeight=0.3202, interpMode=RCIM_Cubic, leaveTangentWeight=0.1067, tangentMode=RCTM_User, time=0.95, value=1}
          {arriveTangentWeight=0.0667, interpMode=RCIM_Cubic, tangentMode=RCTM_User, time=1, value=0}
         ]}}}))) {enabled=false, module=ScaleMeshSize001}
   /SetVariables(Particles.Scale2:Vector3f=&Multiply_VectorByFloat(
     Vector=$Particles.Initial.Scale
     Float=&FloatFromCurve(
      FloatCurve=DataInterface() {propertyValues {Curve {keys [
          {interpMode=RCIM_Cubic, leaveTangentWeight=0.0333, tangentMode=RCTM_User, time=0.7, value=0}
          {arriveTangentWeight=0.17, interpMode=RCIM_Cubic, leaveTangentWeight=0.17, tangentMode=RCTM_User, time=0.8, value=1.25}
          <at 0.8476 1.007>
          {arriveTangentWeight=0.2247, interpMode=RCIM_Cubic, leaveTangentWeight=0.1124, tangentMode=RCTM_User, time=0.95, value=1}
          {arriveTangentWeight=0.0167, interpMode=RCIM_Cubic, tangentMode=RCTM_User, time=1, value=0}
         ]}}}
      Scale Curve=&FloatFromCurve(FloatCurve=DataInterface() {propertyValues {Curve {keys [<at 0 1>,<at 1 0>]}}}, CurveIndex=$System.Timeline.NormalizedTime)
     )
    )) {isSetParameters=true, module=SetVariables_#3}
   &UpdateMeshOrientation(Bank Rate=90, Banking Angle Min / Max:Vector2f=<-5 5>, Orientation Method=::NewEnumerator3)
   &GravityForce() {enabled=false}
   &Drag(Drag=0.25, Use Linear Drag=true, Use Rotational Drag=false) {enabled=false}
   &ScaleColor(
    Scale RGB=&VectorFromFloat(Value=&FloatFromCurve(FloatCurve=DataInterface() {propertyValues {Curve {keys [<at 0.75 1>,<at 0.85 0>]}}}))
    ScaleA=true
    ScaleRGB=true
   )
   &VortexVelocity(
    Module.Use Influence Falloff=false
    Velocity Amount:NiagaraFloat=&FloatFromCurve(
     FloatCurve=DataInterface() {propertyValues {Curve {keys [<at 0.5 1>,<at 0.8 0>]}}}
     Scale Curve=&Multiply_Float(A=20000, B=$Particles.RandSign)
    )
    Vortex Origin:NiagaraPosition=$User.SplineUpdatedEndPos
   )
   &PointAttractionForce(
    Attraction Radius=10000000
    Attraction Strength=20000
    Attractor Position:NiagaraPosition=&LerpPosition(
     A=$User.SplineUpdatedStartPos
     B=$User.SplineUpdatedEndPos
     Alpha=&FloatFromCurve(FloatCurve=DataInterface() {propertyValues {Curve {keys [{tangentMode=RCTM_User, time=0, value=1}, <at 0.3333 1>, <at 0.6667 0>, {tangentMode=RCTM_User, time=1, value=0}]}}})
    )
    Kill Within Radius=false
    Module.Use Falloff=true
    UseAttractorOffset=false
   ) {enabled=false}
   &CurlNoiseForce(
    Noise Frequency=&FloatFromCurve(FloatCurve=DataInterface() {propertyValues {Curve {keys [<at 0.5 120>,<at 0.8 10>]}}})
    Noise Strength=&FloatFromCurve(FloatCurve=DataInterface() {propertyValues {Curve {keys [<at 0.5 0.5>,<at 0.8 1>]}}}, Scale Curve=10000)
    Pan Noise=false
   )
   &PointAttractionForce(
    Attraction Radius=1000000000
    Attraction Strength=&FloatFromCurve(
     FloatCurve=DataInterface() {propertyValues {Curve {keys [{tangentMode=RCTM_User, time=0.5, value=0}, {tangentMode=RCTM_User, time=0.8, value=1}]}}}
     Scale Curve=1000
    )
    Attractor Position:NiagaraPosition=$User.SplineUpdatedStartPos
    Kill Within Radius=false
    Module.Use Falloff=true
    Position To Attract:NiagaraPosition=&AddVectorToPosition(Position=$Particles.Position, Vector=<0 0 0>)
    UseAttractorOffset=false
   ) {module=PointAttractionForce001}
   &SolveForcesAndVelocity(Clamp Velocity=false, Limit Acceleration=false)
   &DynamicMaterialParameters(
    Index 0 Param 1=&RandomRangeFloat(Minimum=0, Maximum=10)
    Index 0 Param 2=&RandomRangeFloat(Minimum=3, Maximum=3.5)
    Index 0 Param 3=&FloatFromCurve(
     FloatCurve=DataInterface() {propertyValues {Curve {keys [
         {interpMode=RCIM_Cubic, tangentMode=RCTM_User, time=0, value=0}
         {interpMode=RCIM_Cubic, tangentMode=RCTM_User, time=0.1, value=1}
         {arriveTangentWeight=0.3202, interpMode=RCIM_Cubic, leaveTangentWeight=0.1067, tangentMode=RCTM_User, time=0.75, value=1}
         {arriveTangentWeight=0.0667, interpMode=RCIM_Cubic, tangentMode=RCTM_User, time=0.85, value=0}
        ]}}}
     Scale Curve=0.725
    )
    Index 1 Param 1=&RandomRangeFloat(Minimum=0, Maximum=3)
    Index 1 Param 2=&RandomRangeFloat(Minimum=0.9, Maximum=1.25)
    Module.Index0Param0Write=true
    Module.Index0Param1Write=true
    Module.Index0Param2Write=true
    Module.Index0Param3Write=true
    Module.Index1Param0Write=true
    Module.Index1Param1Write=true
    Module.Index1Param2Write=true
    Module.Index1Param3Write=true
    Write Parameter Index 1=true
   )
   &KillParticlesInVolume(Origin Offset:NiagaraPosition=$User.SplineUpdatedEndPos, Sphere Radius=10)
   /SetVariables(
    Particles.Flare Color:LinearColor=&ColorFromCurve(ColorCurve=DataInterface() {propertyValues {
       AlphaCurve {keys [
         {arriveTangent=57.0191, interpMode=RCIM_Cubic, leaveTangent=57.0191, tangentMode=RCTM_User, time=0.7899, value=0}
         {interpMode=RCIM_Cubic, leaveTangent=-34.6908, tangentMode=RCTM_Break, time=0.8136, value=1}
         {interpMode=RCIM_Cubic, leaveTangent=27.2465, time=0.8342, value=0}
        ]}
       BlueCurve {keys [<at 0.7631 0 RCIM_Cubic>,<at 0.9252 0 RCIM_Cubic>]}
       GreenCurve {keys [<at 0.7631 0 RCIM_Cubic>,<at 0.9252 0 RCIM_Cubic>]}
       RedCurve {keys [<at 0.7631 2 RCIM_Cubic>,<at 0.9252 1 RCIM_Cubic>]}
      }})
    Particles.flare dynamic:Vector4f=<1 0 0 0>
   ) {isSetParameters=true}
   /SetVariables(Particles.LerpPosition:NiagaraPosition=&LerpPosition(
     A=$Particles.Position
     B=$User.SplineUpdatedStartPos
     Alpha=&FloatFromCurve(FloatCurve=DataInterface() {propertyValues {Curve {keys [
          {
           arriveTangentWeight=0.0497
           interpMode=RCIM_Cubic
           leaveTangentWeight=0.0497
           tangentMode=RCTM_User
           tangentWeightMode=RCTWM_WeightedBoth
           time=0.85
           value=0
          }
          {
           arriveTangentWeight=0.0558
           interpMode=RCIM_Cubic
           leaveTangentWeight=0.0558
           tangentMode=RCTM_User
           tangentWeightMode=RCTWM_WeightedBoth
           time=1
           value=1
          }
         ]}}})
    )) {isSetParameters=true, module=SetVariables_#4}
  ]
  rendererClasses [&NiagaraMeshRendererProperties, &NiagaraSpriteRendererProperties]
  renderers [
   {rendererClass=&NiagaraMeshRendererProperties, rendererIndex=0}
   {rendererClass=&NiagaraMeshRendererProperties, rendererIndex=1}
   {rendererClass=&NiagaraSpriteRendererProperties, rendererIndex=2}
  ]
 }
 Flame {
  emitterUpdateScript [&EmitterState(Life Cycle Mode=::NewEnumerator1, Loop Behavior=::NewEnumerator1, Loop Duration=0.2, UseLoopDelay=false)]
  particleSpawnScript [
   &InitializeParticle(
    Color Maximum=<rgba 1 1 1 0.5>
    Color Minimum=<rgba 0.5596 0.5365 1 1>
    Color Mode=::NewEnumerator2
    Lifetime Max=1.2
    Lifetime Min=0.7
    Lifetime Mode=::NewEnumerator1
    Sprite Rotation Mode=::NewEnumerator2
    Sprite Size Mode=::NewEnumerator1
    Uniform Sprite Size Max=150
    Uniform Sprite Size Min=80
    UsePositionOffset=false
    Write Material Random=true
   )
   /SetVariables(Particles.RandomFrame=&RandomRangeInt(Minimum=0, Maximum=63)) {isSetParameters=true}
   &DynamicMaterialParameters(
    Index 0 Param 1=&RandomRangeFloat(Minimum=0.2, Maximum=0.5)
    Module.Param0WriteEnabled=true
    Module.Param1WriteEnabled=true
    Module.Param2WriteEnabled=true
    Module.Param3WriteEnabled=true
   )
   &AddVelocityFromPoint(Module.Use Falloff Distance=false, Velocity Strength=&RandomRangeFloat(Minimum=40, Maximum=80))
   &CameraOffset(Camera Offset Amount=&RandomRangeFloat(Minimum=0, Maximum=10))
  ]
  particleUpdateScript [
   &ParticleState()
   &GravityForce(Gravity=&RandomRangeVector(Minimum=<-50 -50 -100>, Maximum=<50 50 -200>))
   &FN_SubUVAnimation(End Frame=&Add_Integer(A=$Particles.RandomFrame, B=63), Start Frame=$Particles.RandomFrame, SubUV Loop Count=0.75)
   &Drag(Drag=&RandomRangeFloat(Minimum=4, Maximum=6), Use Linear Drag=true, Use Rotational Drag=false)
   &CurlNoiseForce(Noise Frequency=20, Noise Strength=&RandomRangeFloat(Minimum=200, Maximum=400), Pan Noise=false)
   &SolveForcesAndVelocity(Clamp Velocity=false, Limit Acceleration=false)
   &ScaleColor(
    Scale Alpha=&Multiply_Float(
     A=&FloatFromCurve(
      FloatCurve=DataInterface() {propertyValues {Curve {keys [
          {arriveTangent=2.5779, interpMode=RCIM_Cubic, leaveTangent=2.5779, tangentMode=RCTM_User, time=0, value=0.4}
          {arriveTangent=0.3312, interpMode=RCIM_Cubic, leaveTangent=-0.2993, tangentMode=RCTM_Break, time=0.2, value=1}
          {arriveTangent=-0.6682, interpMode=RCIM_Cubic, leaveTangent=-0.6682, tangentMode=RCTM_User, time=0.4, value=0.15}
          {arriveTangent=-0.2059, interpMode=RCIM_Cubic, leaveTangent=-0.2059, tangentMode=RCTM_User, time=1, value=0}
         ]}}}
      Scale Curve=&RandomRangeFloat(Minimum=0.3, Maximum=0.6)
     )
     B=&FNTimeOfDayScalar(Night Scale=1, Morning Scale=0.3, Evening Scale=0.5)
    )
    Scale RGB=&MakeVectorFromLinearColorRGB(Color=&ColorFromCurve(ColorCurve=DataInterface() {propertyValues {
        AlphaCurve {keys [<at 0 1 RCIM_Cubic>]}
        BlueCurve {keys [<at 0 0.0065>,<at 0.1619 0>,<at 0.546 0 RCIM_Cubic>,<at 0.9165 0 RCIM_Cubic>]}
        GreenCurve {keys [<at 0 0>,<at 0.1619 0.0251>,<at 0.546 0.0019 RCIM_Cubic>,<at 0.9165 0.0251 RCIM_Cubic>]}
        RedCurve {keys [<at 0 0.3854>,<at 0.1619 0.6198>,<at 0.546 0.026 RCIM_Cubic>,<at 0.9165 0.6198 RCIM_Cubic>]}
       }}))
    ScaleA=true
    ScaleRGB=true
   ) {module=ScaleColor001}
   &ScaleSpriteSize(Scale Factor=&Vector2DFromFloat(Value=&FloatFromCurve(FloatCurve=DataInterface() {propertyValues {Curve {keys [{tangentMode=RCTM_User, time=0, value=0.8}, {tangentMode=RCTM_User, time=1, value=1}]}}}))) {module=ScaleSpriteSize001}
  ]
  rendererClasses [&NiagaraSpriteRendererProperties]
  renderers [{rendererClass=&NiagaraSpriteRendererProperties}]
 }
 Glow {
  emitterUpdateScript [
   &EmitterState(Life Cycle Mode=::NewEnumerator1, Loop Behavior=::NewEnumerator1, Loop Duration Mode=::NewEnumerator1, UseLoopDelay=false)
   &SpawnBurst_Instantaneous(Spawn Count=1, Use Spawn Probability=false)
  ]
  particleSpawnScript [
   &InitializeParticle(
    Color=<rgba 3 0 0 1>
    Lifetime=0.5
    Sprite Size Mode=::NewEnumerator3
    Uniform Sprite Size=80
    UsePositionOffset=false
    Write Material Random=true
   )
   /SetVariables() {isSetParameters=true}
  ]
  particleUpdateScript [
   &ParticleState()
   &self:WaveAxis(EndPos:NiagaraPosition=$User.SplineUpdatedEndPos, StartPos:NiagaraPosition=$User.SplineUpdatedStartPos) {module=WaveAxis}
   /SetVariables(Particles.BeamStartDissolve:NiagaraFloat=&ClampFloat(Float=&Add_Float(A=$Particles.BeamStartDissolve, B=&Multiply_Float(A=$Engine.DeltaTime, B=4)))) {isSetParameters=true}
   /SetVariables(Particles.SinePosition:Vector3f=&Add_Vector(
     A=&Multiply_Vector(
      A=$Particles.Y
      B=&VectorFromFloat(Value=&Sine(
        Normalized Angle=&Add_Float(
         A=&Multiply_Float(A=$System.NormalizedLoopAge, B=-1)
         B=&Multiply_Float(A=$Engine.Time, B=0.2)
        )
        Period=0.5
        Scale=15
        Bias=0
       ))
     )
     B=&Multiply_Vector(
      A=$Particles.Z
      B=&VectorFromFloat(Value=&Cosine(
        Normalized Angle=&Add_Float(
         A=&Multiply_Float(A=$System.NormalizedLoopAge, B=-1)
         B=&Multiply_Float(A=$Engine.Time, B=0.2)
        )
        Period=0.5
        Scale=15
        Bias=0
       ))
     )
    )) {isSetParameters=true}
   /SetVariables(Particles.MultiplySinePosition:Vector3f=&Multiply_Vector(
     A=$Particles.SinePosition
     B=&VectorFromFloat(Value=&FloatFromCurve(
       FloatCurve=DataInterface() {propertyValues {Curve {keys [
           {interpMode=RCIM_Cubic, tangentMode=RCTM_User, time=0, value=0}
           {interpMode=RCIM_Cubic, tangentMode=RCTM_User, time=0.5, value=1}
           {tangentMode=RCTM_User, time=1, value=0}
          ]}}}
       CurveIndex=$System.NormalizedLoopAge
      ))
    )) {isSetParameters=true}
   /SetVariables(Particles.NewPosition:Vector3f=&Add_Vector(
     A=$Particles.MultiplySinePosition
     B=&Lerp_Vector(
      A=Linked() {linkedVariable {name=User.SplineUpdatedStartPos, type=NiagaraPosition}}
      B=&Add_Vector(A=Linked() {linkedVariable {name=User.SplineUpdatedEndPos, type=NiagaraPosition}}, B=<0 0 0>)
      Alpha=$Particles.BeamStartDissolve
     )
    )) {isSetParameters=true}
   &SolveForcesAndVelocity(Clamp Velocity=false, Limit Acceleration=false)
   &ScaleColor(
    Scale Alpha=&FNTimeOfDayScalar(Night Scale=1, Morning Scale=0.3, Evening Scale=0.5)
    ScaleA=true
    ScaleRGB=true
   )
   /SetVariables(
    Particles.FlareColor2:LinearColor=&MultiplyLinearColor(
     Linear Color=<rgba 1 0.2578 0.0104 0.3>
     Scale Factor=&MakeVector4_FromFloat(Float=&FNTimeOfDayScalar(Night Scale=1, Morning Scale=0.3, Evening Scale=0.5))
    )
    Particles.FlareSize2:Vector2f=&Vector2DFromFloat(Value=100)
   ) {isSetParameters=true}
   &GenerateLocationEvent(Module.Use Event Send Rate=true, Position To Send:Vector3f=$Particles.NewPosition)
   &DoOnce(Reset=true)
  ]
  rendererClasses [&NiagaraSpriteRendererProperties]
  renderers [
   {rendererClass=&NiagaraSpriteRendererProperties}
   {rendererClass=&NiagaraSpriteRendererProperties, rendererIndex=1}
  ]
 }
 SplineElectric {
  emitterSpawnScript [/Game/Effects/Niagara/Modules/Partition/FN_PartitionAttributes(Number of Bone Sets=1, Particles Per Partition=12)]
  emitterUpdateScript [
   &EmitterState(Life Cycle Mode=::NewEnumerator1, Loop Behavior=::NewEnumerator1, Loop Duration Mode=::NewEnumerator1, UseLoopDelay=false)
   &SpawnBurst_Instantaneous(Spawn Count=$Emitter.FN_PartitionAttributes.Spawn Total, Use Spawn Probability=false)
  ]
  particleSpawnScript [
   /Game/Effects/Niagara/Modules/Partition/FN_Particle_partition_v5/FN_Particle_Partition_v5(Enable Ribbon ID=true)
   &InitializeParticle(Ribbon Width=50, Ribbon Width Mode=::NewEnumerator1, UsePositionOffset=false, Write Material Random=true)
   /SetVariables(Particles.BeamEndDissolve=1) {isSetParameters=true, module=SetVariables_#5}
  ]
  particleUpdateScript [
   &ParticleState(Kill Particles When Lifetime Has Elapsed=false, Let Infinitely Lived Particles Die When Emitter Deactivates=true)
   /SetVariables(
    Particles.BeamEndDissolve:NiagaraFloat=&ClampFloat(Float=&Subtract_Float(
      A=$Particles.BeamEndDissolve
      B=&Multiply_Float(
       A=$Engine.DeltaTime
       B=&MakeCustomFloatFromBool(
        Bool=&SetBoolByFloatComparison(A=$User.RemoveBeam, B=1, Comparison Type=::NewEnumerator4)
        True Float=4
        False Float=0
       )
      )
     ))
    Particles.BeamStartDissolve:NiagaraFloat=&ClampFloat(Float=&Add_Float(A=$Particles.BeamStartDissolve, B=&Multiply_Float(A=$Engine.DeltaTime, B=2)))
   ) {isSetParameters=true, module=SetVariables_#6}
   &DynamicMaterialParameters(
    Index 0 Param 1=&FloatFromCurve(
     FloatCurve=DataInterface() {propertyValues {Curve {keys [{tangentMode=RCTM_User, time=0, value=0}, {tangentMode=RCTM_User, time=1, value=1}]}}}
     CurveIndex=&Multiply_Float(A=$Particles.BeamStartDissolve, B=$Particles.BeamEndDissolve)
    )
    Index 0 Param 2=1.02
    Index 0 Param 4=&FNTimeOfDayScalar(Night Scale=0.5, Morning Scale=0.15, Day Scale=0.15, Evening Scale=0.15)
    Module.Param0WriteEnabled=true
    Module.Param1WriteEnabled=true
    Module.Param2WriteEnabled=false
    Module.Param3WriteEnabled=true
   )
   &SystemLocation(MaskSpawnGroup=false)
   /Game/Effects/Niagara/Modules/Curves/FN_HermiteCubicSpline_v1/FN_HermiteCubicSpline(
    Position Blend:NiagaraFloat=$Particles.Initial.Partition.NormalizedID
    Position End:NiagaraPosition=$User.SplineUpdatedStartPos
    Position Start:Vector3f=&AddVectorToPosition(
     Position=$User.SplineUpdatedEndPos
     Vector=&MakeVector(
      X=&Waveform(Global Input Position=1000, "Waveform [1]"=::NewEnumerator12, "[1] Input Position"=1000, "[1] Frequency"=4, "[1] Amplitude Min/Max"=<-10 10>)
      Y=&Waveform("Waveform [1]"=::NewEnumerator12, "[1] Frequency"=4, "[1] Amplitude Min/Max"=<-10 10>)
      Z=&Waveform("Waveform [1]"=::NewEnumerator12, "[1] Frequency"=2, "[1] Amplitude Min/Max"=<-10 10>)
     )
    )
    Write Particles Position=true
   )
   &self:WaveAxis(EndPos:NiagaraPosition=$User.SplineUpdatedEndPos, StartPos:NiagaraPosition=$User.SplineUpdatedStartPos) {module=WaveAxis}
   /SetVariables(Particles.SinePosition:Vector3f=&Add_Vector(
     A=&Multiply_Vector(
      A=$Particles.Y
      B=&VectorFromFloat(Value=&Sine(
        Normalized Angle=&Add_Float(
         A=&Multiply_Float(A=$Particles.Partition.NormalizedID, B=-1)
         B=&Multiply_Float(A=$Engine.Time, B=0.2)
        )
        Period=0.5
        Scale=15
        Bias=0
       ))
     )
     B=&Multiply_Vector(
      A=$Particles.Z
      B=&VectorFromFloat(Value=&Cosine(
        Normalized Angle=&Add_Float(
         A=&Multiply_Float(A=$Particles.Partition.NormalizedID, B=-1)
         B=&Multiply_Float(A=$Engine.Time, B=0.2)
        )
        Period=0.5
        Scale=15
        Bias=0
       ))
     )
    )) {isSetParameters=true, module=SetVariables_#7}
   /SetVariables(Particles.MultiplySinePosition:Vector3f=&Multiply_Vector(
     A=$Particles.SinePosition
     B=&VectorFromFloat(Value=&FloatFromCurve(
       FloatCurve=DataInterface() {propertyValues {Curve {keys [
           {interpMode=RCIM_Cubic, tangentMode=RCTM_User, time=0, value=0}
           {interpMode=RCIM_Cubic, tangentMode=RCTM_User, time=0.5, value=1}
           {tangentMode=RCTM_User, time=1, value=0}
          ]}}}
       CurveIndex=$Particles.Partition.NormalizedID
      ))
    )) {isSetParameters=true, module=SetVariables_#8}
   /SetVariables(Particles.Position=&Add_Vector(A=$Particles.MultiplySinePosition, B=$Output.FN_HermiteCubicSpline.Position)) {isSetParameters=true, module=SetVariables_#9}
   /SetVariables(Particles.EscapeDistance:NiagaraFloat=&Divide_Float(
     A=&VectorLength(Vector=&Subtract_Vector(
       A=Linked() {linkedVariable {name=User.SplineUpdatedEndPos, type=NiagaraPosition}}
       B=Linked() {linkedVariable {name=User.SplineUpdatedStartPos, type=NiagaraPosition}}
      ))
     B=1280
    )) {isSetParameters=true, module=SetVariables_#10}
   &ScaleRibbonWidth(Ribbon Width Scale=&FloatFromCurve(
     FloatCurve=DataInterface() {propertyValues {Curve {keys [{tangentMode=RCTM_User, time=0, value=0}, {tangentMode=RCTM_User, time=1, value=1}]}}}
     CurveIndex=$Particles.BeamEndDissolve
    )) {module=ScaleRibbonWidth002}
   &ScaleRibbonWidth(Ribbon Width Scale=&FloatFromCurve(
     FloatCurve=DataInterface() {propertyValues {Curve {keys [<at 0.7 1>, {tangentMode=RCTM_User, time=0.9, value=0.8}, {tangentMode=RCTM_User, time=1, value=0.5}]}}}
     CurveIndex=$Particles.EscapeDistance
    )) {module=ScaleRibbonWidth001}
   &ScaleColor(
    Scale Alpha=&FloatFromCurve(
     FloatCurve=DataInterface() {propertyValues {Curve {keys [{tangentMode=RCTM_User, time=0, value=0}, {tangentMode=RCTM_User, time=1, value=1}]}}}
     CurveIndex=$Particles.BeamEndDissolve
     Scale Curve=0.5
    )
    Scale RGB=<1 0.5 0.5>
    ScaleA=true
    ScaleRGB=true
   ) {module=ScaleColor001}
   &ScaleColor(
    Scale Alpha=&Multiply_Float(
     A=&FloatFromCurve(
      FloatCurve=DataInterface() {propertyValues {Curve {keys [<at 0.7 1>, {tangentMode=RCTM_User, time=0.9, value=0.8}, {tangentMode=RCTM_User, time=1, value=0.5}]}}}
      CurveIndex=$Particles.EscapeDistance
     )
     B=&FNTimeOfDayScalar(Night Scale=0.4, Morning Scale=0.3, Evening Scale=0.5)
    )
    Scale RGB=&VectorFromFloat(Value=&MakeCustomFloatFromBool(
      Bool=&SetBoolByFloatComparison(
       A=&OneMinusFloat(Float=&FloatFromCurve(
         FloatCurve=DataInterface() {propertyValues {Curve {keys [<at 0.7 1>, {tangentMode=RCTM_User, time=0.9, value=0.8}, {tangentMode=RCTM_User, time=1, value=0.5}]}}}
         CurveIndex=$Particles.EscapeDistance
        ))
       B=0.2
      )
      True Float=&RandomRangeFloat(Minimum=0.5, Maximum=2, Evaluation Type=::NewEnumerator1)
      False Float=1
     ))
    ScaleA=true
    ScaleRGB=true
   )
   &ScaleRibbonWidth(Ribbon Width Scale=&FloatFromCurve(
     FloatCurve=DataInterface() {propertyValues {Curve {keys [{arriveTangent=-1.5292, interpMode=RCIM_Cubic, leaveTangent=-1.5292, tangentMode=RCTM_User, time=0, value=2}, <at 1 0.35 RCIM_Cubic>]}}}
     CurveIndex=&ReturnNormalizedExecIndex()
    ))
  ]
  rendererClasses [&NiagaraRibbonRendererProperties]
  renderers [{rendererClass=&NiagaraRibbonRendererProperties}]
 }
 SplineElectricPurple {
  bEnabled=false
  emitterSpawnScript [&FN_PartitionAttributes(Number of Bone Sets=1, Particles Per Partition=12)]
  emitterUpdateScript [
   &EmitterState(Life Cycle Mode=::NewEnumerator1, Loop Behavior=::NewEnumerator1, Loop Duration Mode=::NewEnumerator1, UseLoopDelay=false)
   &SpawnBurst_Instantaneous(Spawn Count=$Emitter.FN_PartitionAttributes.Spawn Total, Use Spawn Probability=false)
  ]
  particleSpawnScript [
   &FN_Particle_Partition_v5(Enable Ribbon ID=true)
   &InitializeParticle(Ribbon Width=50, Ribbon Width Mode=::NewEnumerator1, UsePositionOffset=false, Write Material Random=true)
   /SetVariables(Particles.BeamEndDissolve=1) {isSetParameters=true, module=SetVariables_#5}
  ]
  particleUpdateScript [
   &ParticleState(Kill Particles When Lifetime Has Elapsed=false, Let Infinitely Lived Particles Die When Emitter Deactivates=true)
   /SetVariables(
    Particles.BeamEndDissolve:NiagaraFloat=&ClampFloat(Float=&Subtract_Float(
      A=$Particles.BeamEndDissolve
      B=&Multiply_Float(
       A=$Engine.DeltaTime
       B=&MakeCustomFloatFromBool(
        Bool=&SetBoolByFloatComparison(A=$User.RemoveBeam, B=1, Comparison Type=::NewEnumerator4)
        True Float=4
        False Float=0
       )
      )
     ))
    Particles.BeamStartDissolve:NiagaraFloat=&ClampFloat(Float=&Add_Float(A=$Particles.BeamStartDissolve, B=&Multiply_Float(A=$Engine.DeltaTime, B=2)))
   ) {isSetParameters=true, module=SetVariables_#6}
   &DynamicMaterialParameters(
    Index 0 Param 1=&FloatFromCurve(
     FloatCurve=DataInterface() {propertyValues {Curve {keys [{tangentMode=RCTM_User, time=0, value=0}, {tangentMode=RCTM_User, time=1, value=1}]}}}
     CurveIndex=&Multiply_Float(A=$Particles.BeamStartDissolve, B=$Particles.BeamEndDissolve)
    )
    Index 0 Param 2=1.02
    Index 0 Param 4=&FNTimeOfDayScalar(Night Scale=0.5, Morning Scale=0.15, Day Scale=0.15, Evening Scale=0.15)
    Module.Param0WriteEnabled=true
    Module.Param1WriteEnabled=true
    Module.Param2WriteEnabled=false
    Module.Param3WriteEnabled=true
   )
   &SystemLocation(MaskSpawnGroup=false)
   &FN_HermiteCubicSpline(
    Position Blend:NiagaraFloat=$Particles.Initial.Partition.NormalizedID
    Position End:NiagaraPosition=$User.SplineUpdatedStartPos
    Position Start:Vector3f=&AddVectorToPosition(
     Position=$User.SplineUpdatedEndPos
     Vector=&MakeVector(
      X=&Waveform(Global Input Position=1000, "Waveform [1]"=::NewEnumerator12, "[1] Input Position"=1000, "[1] Frequency"=4, "[1] Amplitude Min/Max"=<-10 10>)
      Y=&Waveform("Waveform [1]"=::NewEnumerator12, "[1] Frequency"=4, "[1] Amplitude Min/Max"=<-10 10>)
      Z=&Waveform("Waveform [1]"=::NewEnumerator12, "[1] Frequency"=2, "[1] Amplitude Min/Max"=<-10 10>)
     )
    )
    Write Particles Position=true
   )
   &self:WaveAxis(EndPos:NiagaraPosition=$User.SplineUpdatedEndPos, StartPos:NiagaraPosition=$User.SplineUpdatedStartPos) {module=WaveAxis}
   /SetVariables(Particles.SinePosition:Vector3f=&Add_Vector(
     A=&Multiply_Vector(
      A=$Particles.Y
      B=&VectorFromFloat(Value=&Sine(
        Normalized Angle=&Add_Float(
         A=&Multiply_Float(A=$Particles.Partition.NormalizedID, B=-1)
         B=&Multiply_Float(A=$Engine.Time, B=0.2)
        )
        Period=0.5
        Scale=15
        Bias=0
       ))
     )
     B=&Multiply_Vector(
      A=$Particles.Z
      B=&VectorFromFloat(Value=&Cosine(
        Normalized Angle=&Add_Float(
         A=&Multiply_Float(A=$Particles.Partition.NormalizedID, B=-1)
         B=&Multiply_Float(A=$Engine.Time, B=0.2)
        )
        Period=0.5
        Scale=15
        Bias=0
       ))
     )
    )) {isSetParameters=true, module=SetVariables_#7}
   /SetVariables(Particles.MultiplySinePosition:Vector3f=&Multiply_Vector(
     A=$Particles.SinePosition
     B=&VectorFromFloat(Value=&FloatFromCurve(
       FloatCurve=DataInterface() {propertyValues {Curve {keys [
           {interpMode=RCIM_Cubic, tangentMode=RCTM_User, time=0, value=0}
           {interpMode=RCIM_Cubic, tangentMode=RCTM_User, time=0.5, value=1}
           {tangentMode=RCTM_User, time=1, value=0}
          ]}}}
       CurveIndex=$Particles.Partition.NormalizedID
      ))
    )) {isSetParameters=true, module=SetVariables_#8}
   /SetVariables(Particles.Position=&Add_Vector(A=$Particles.MultiplySinePosition, B=$Output.FN_HermiteCubicSpline.Position)) {isSetParameters=true, module=SetVariables_#9}
   /SetVariables(Particles.EscapeDistance:NiagaraFloat=&Divide_Float(
     A=&VectorLength(Vector=&Subtract_Vector(
       A=Linked() {linkedVariable {name=User.SplineUpdatedEndPos, type=NiagaraPosition}}
       B=Linked() {linkedVariable {name=User.SplineUpdatedStartPos, type=NiagaraPosition}}
      ))
     B=1280
    )) {isSetParameters=true, module=SetVariables_#10}
   &ScaleRibbonWidth(Ribbon Width Scale=&FloatFromCurve(
     FloatCurve=DataInterface() {propertyValues {Curve {keys [{tangentMode=RCTM_User, time=0, value=0}, {tangentMode=RCTM_User, time=1, value=1}]}}}
     CurveIndex=$Particles.BeamEndDissolve
    )) {module=ScaleRibbonWidth002}
   &ScaleRibbonWidth(Ribbon Width Scale=&FloatFromCurve(
     FloatCurve=DataInterface() {propertyValues {Curve {keys [<at 0.7 1>, {tangentMode=RCTM_User, time=0.9, value=0.8}, {tangentMode=RCTM_User, time=1, value=0.5}]}}}
     CurveIndex=$Particles.EscapeDistance
    )) {module=ScaleRibbonWidth001}
   &ScaleColor(
    Scale Alpha=&FloatFromCurve(
     FloatCurve=DataInterface() {propertyValues {Curve {keys [{tangentMode=RCTM_User, time=0, value=0}, {tangentMode=RCTM_User, time=1, value=1}]}}}
     CurveIndex=$Particles.BeamEndDissolve
    )
    ScaleA=true
    ScaleRGB=false
   ) {module=ScaleColor001}
   &ScaleColor(
    Scale Alpha=&Multiply_Float(
     A=&FloatFromCurve(
      FloatCurve=DataInterface() {propertyValues {Curve {keys [<at 0.7 1>, {tangentMode=RCTM_User, time=0.9, value=0.8}, {tangentMode=RCTM_User, time=1, value=0.5}]}}}
      CurveIndex=$Particles.EscapeDistance
     )
     B=&FNTimeOfDayScalar(Night Scale=0.4, Morning Scale=0.3, Evening Scale=0.5)
    )
    Scale RGB=&VectorFromFloat(Value=&MakeCustomFloatFromBool(
      Bool=&SetBoolByFloatComparison(
       A=&OneMinusFloat(Float=&FloatFromCurve(
         FloatCurve=DataInterface() {propertyValues {Curve {keys [<at 0.7 1>, {tangentMode=RCTM_User, time=0.9, value=0.8}, {tangentMode=RCTM_User, time=1, value=0.5}]}}}
         CurveIndex=$Particles.EscapeDistance
        ))
       B=0.2
      )
      True Float=&RandomRangeFloat(Minimum=0.5, Maximum=2, Evaluation Type=::NewEnumerator1)
      False Float=1
     ))
    ScaleA=true
    ScaleRGB=true
   )
   &ScaleRibbonWidth(Ribbon Width Scale=&FloatFromCurve(
     FloatCurve=DataInterface() {propertyValues {Curve {keys [{arriveTangent=-1.5292, interpMode=RCIM_Cubic, leaveTangent=-1.5292, tangentMode=RCTM_User, time=0, value=2}, <at 1 0.35 RCIM_Cubic>]}}}
     CurveIndex=&ReturnNormalizedExecIndex()
    ))
  ]
  rendererClasses [&NiagaraRibbonRendererProperties]
  renderers [{rendererClass=&NiagaraRibbonRendererProperties}]
 }
}
summary {
 systemName=NS_LifeDrain_Beam_Vampire_Loop
 emitters [
  {emitterName=SplineElectric, rendererClasses [&NiagaraRibbonRendererProperties]}
  {emitterName=Glow, rendererClasses [&NiagaraSpriteRendererProperties]}
  {emitterName=Flame, rendererClasses [&NiagaraSpriteRendererProperties]}
  {emitterName=Bats, rendererClasses [&NiagaraMeshRendererProperties]}
  {bEnabled=false, emitterName=SplineElectricPurple, rendererClasses [&NiagaraRibbonRendererProperties]}
  {
   bEnabled=false
   emitterName=BirdBat001
   rendererClasses [&NiagaraMeshRendererProperties, &NiagaraSpriteRendererProperties]
  }
 ]
 userVariables [
  User.AlphaMult {type=NiagaraFloat, defaultValue=1, description=""}
  User.CorrectDirection {type=NiagaraFloat, defaultValue=1, description=""}
  User.PhysMatIndex {type=NiagaraInt32}
  User.RemoveBeam {type=NiagaraFloat, defaultValue=0, description=""}
  User.SplineUpdatedEndPos {type=NiagaraPosition, description="", defaultValue {}}
  User.SplineUpdatedEndTangent {type=Vector3f, description="", defaultValue {struct=Vector3f, value=<0 0 0>}}
  User.SplineUpdatedStartPos {type=NiagaraPosition, description="", defaultValue {}}
 ]
}
