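# Niagara ambient effect for the Vampire Boss that sheds faint, dark red-to-black smoke sprites from a skeletal mesh supplied as a user parameter.
# Particle_Smoke scales with owner velocity over 300, so faster movement spawns more, shorter-lived puffs.
# Sprites inherit sampled mesh velocity, slow under drag, and grow and fade with SubUV animation.
# The Empty emitter, which feeds owner speed into a dynamic material parameter, is disabled.
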NiagaraSystem /VampireBoss/VFX/NS_VampireBossAmbient
dependencies {
 usedDataInterfaces [&NiagaraDataInterfaceCurve]
 usedDynamicInputs [&RandomRangeVector]
 usedModules [
  &self:SystemSpawnScript.SystemScriptSource.SystemScriptGraph.NiagaraNodeAssignment_0.SetVariables_#1
  &self:SystemSpawnScript.SystemScriptSource.SystemScriptGraph.NiagaraNodeAssignment_1.SetVariables_#2
  &SystemState
  &self:SystemSpawnScript.SystemScriptSource.SystemScriptGraph.NiagaraNodeAssignment_2.SetVariables_#3
 ]
}
emitters {
 Empty {
  bEnabled=false
  emitterUpdateScript [
   &EmitterState()
   &SpawnRate(SpawnRate=20, Use Spawn Probability=false)
  ]
  particleSpawnScript [
   &InitializeParticle(
    Color Mode=::NewEnumerator0
    Lifetime=5
    Position Offset=<0 0 100>
    Sprite Size Mode=::NewEnumerator3
    Uniform Sprite Size=100
    UsePositionOffset=true
    Write Material Random=true
   )
   &CameraOffset()
   &CameraOffset(Camera Offset Amount=25) {module=CameraOffset001}
  ]
  particleUpdateScript [
   &ParticleState()
   &DynamicMaterialParameters(
    Index 0 Param 1=&VectorLength(Vector=&Divide_Vector(A:Vector3f=$Engine.Owner.Velocity, B:Vector3f=&VectorFromFloat(Value=300)))
    Module.Index0Param0Write=true
    Module.Index0Param1Write=true
    Module.Index0Param2Write=true
    Module.Index0Param3Write=true
   )
  ]
  rendererClasses [&NiagaraSpriteRendererProperties]
  renderers [{rendererClass=&NiagaraSpriteRendererProperties}]
 }
 Particle_Smoke {
  emitterUpdateScript [
   &EmitterState()
   &SpawnRate(
    SpawnRate=&Lerp_Float(
     A=&RandomRangeFloat(Minimum=9, Maximum=9)
     B=&RandomRangeFloat(Minimum=22, Maximum=22)
     Alpha=&FloatFromCurve(
      FloatCurve=DataInterface() {propertyValues {Curve {keys [<at 0 0>,<at 1 1>]}}}
      CurveIndex=&VectorLength(Vector=&Divide_Vector(A:Vector3f=$Engine.Owner.Velocity, B:Vector3f=&VectorFromFloat(Value=300)))
     )
    )
    Use Spawn Probability=false
   )
  ]
  particleSpawnScript [
   /SetVariables(Particles.RandCurve:NiagaraFloat=&FloatFromCurve(
     FloatCurve=DataInterface() {propertyValues {Curve {keys [
         {interpMode=RCIM_Cubic, tangentMode=RCTM_User, time=0, value=0}
         {arriveTangent=0.7625, interpMode=RCIM_Cubic, leaveTangent=0.7625, tangentMode=RCTM_User, time=0.8, value=0.2}
         {arriveTangent=7.0993, interpMode=RCIM_Cubic, leaveTangent=7.0993, tangentMode=RCTM_User, time=1, value=1}
        ]}}}
     CurveIndex=&RandomRangeFloat(Minimum=0, Maximum=1)
    )) {isSetParameters=true}
   &InitializeParticle(
    Color=&LerpLinearColors(
     StartColor=<rgba 0.0417 0.0049 0 0.2>
     EndColor=<rgba 0 0 0 0.15>
     LerpFactor=&FloatFromCurve(
      FloatCurve=DataInterface() {propertyValues {Curve {keys [
          {interpMode=RCIM_Cubic, leaveTangentWeight=0.6285, tangentMode=RCTM_Break, tangentWeightMode=RCTWM_WeightedBoth, time=0, value=0}
          {arriveTangent=-0.0925, interpMode=RCIM_Cubic, leaveTangent=-0.0925, tangentMode=RCTM_User, time=0.5, value=1}
         ]}}}
      CurveIndex=&RandomRangeFloat(Minimum=0, Maximum=1)
     )
    )
    Lifetime=&Lerp_Float(
     A=&RandomRangeFloat(Minimum=2, Maximum=1)
     B=&RandomRangeFloat(Minimum=0.35, Maximum=0.2)
     Alpha=&FloatFromCurve(
      FloatCurve=DataInterface() {propertyValues {Curve {keys [<at 0.1 0>,<at 1 1>]}}}
      CurveIndex=&VectorLength(Vector=&Divide_Vector(A:Vector3f=$Engine.Owner.Velocity, B:Vector3f=&VectorFromFloat(Value=300)))
     )
    )
    Sprite Rotation Mode=::NewEnumerator2
    Sprite Size Max=<100 100>
    Sprite Size Min=<30 30>
    Sprite Size Mode=::NewEnumerator2
    Sprite UV Mode=::NewEnumerator3
    UsePositionOffset=false
    Write Material Random=true
   )
   &InitializeParticle(
    Color=&LerpLinearColors(
     StartColor=<rgba 0.0417 0.0006 0 0.1>
     EndColor=<rgba 0 0 0 0.1>
     LerpFactor=&FloatFromCurve(
      FloatCurve=DataInterface() {propertyValues {Curve {keys [
          {interpMode=RCIM_Cubic, tangentMode=RCTM_User, time=0, value=0}
          {arriveTangent=0.7625, interpMode=RCIM_Cubic, leaveTangent=0.7625, tangentMode=RCTM_User, time=0.9, value=0.2}
          {arriveTangent=12.0097, interpMode=RCIM_Cubic, leaveTangent=12.0097, tangentMode=RCTM_User, time=1, value=1}
         ]}}}
      CurveIndex=&RandomRangeFloat(Minimum=0, Maximum=1)
     )
    )
    Lifetime Max=1.2
    Lifetime Min=0.9
    Lifetime Mode=::NewEnumerator1
    Sprite Rotation Mode=::NewEnumerator2
    Sprite Size=&Multiply_Vector2D_ByFloat(Float=&Lerp_Float(A=1, B=2, Alpha=$Particles.RandCurve), Vector2D=<36 32>)
    Sprite Size Mode=::NewEnumerator4
    UsePositionOffset=false
    Write Material Random=true
   ) {enabled=false, module=InitializeParticle001}
   &CameraOffset(Camera Offset Amount=&RandomRangeFloat(Minimum=5, Maximum=15))
   &SkeletalMeshLocation(
    Bone Sampling Mode=::NewEnumerator0
    Interpolate Mesh Sampling=false
    MaskSpawnGroup=false
    OverrideRandomSeed=false
    Position Sampling=::NewEnumerator1
    Skeletal Mesh=$User.Skeletal Mesh
    Source Movement Scale:Vector3f=&VectorFromFloat(Value=&FloatFromCurve(FloatCurve=DataInterface() {propertyValues {Curve {keys [<at 0 -1>,<at 1 0>]}}}))
    Use Sampled Mesh Attribute as Ribbon ID=false
   )
   &InheritVelocity(Speed Limit=true, UseSpeedThreshold=false, Velocity Source=$Particles.SkeletalMeshLocation.SampledVelocity)
  ]
  particleUpdateScript [
   &ParticleState()
   &ScaleSpriteSize(Uniform Curve Sprite Scale=DataInterface() {propertyValues {Curve {keys [
        {arriveTangent=0.686, interpMode=RCIM_Cubic, leaveTangent=0.686, tangentMode=RCTM_User, time=0, value=1}
        {arriveTangent=1.2865, interpMode=RCIM_Cubic, leaveTangent=1.2865, tangentMode=RCTM_User, time=1, value=1.5}
       ]}}})
   &SubUVAnimation(SubUV Lookup Index:NiagaraFloat=&Multiply_Float(A=$Particles.NormalizedAge, B=0.25))
   &ScaleColor(
    Scale Alpha=&FloatFromCurve(
     FloatCurve=DataInterface() {propertyValues {Curve {keys [
         {interpMode=RCIM_Cubic, tangentMode=RCTM_User, time=0, value=0}
         {interpMode=RCIM_Cubic, tangentMode=RCTM_User, time=0.4, value=1}
         {interpMode=RCIM_Cubic, tangentMode=RCTM_User, time=0.7, value=1}
         {interpMode=RCIM_Cubic, tangentMode=RCTM_User, time=1, value=0}
        ]}}}
     Scale Curve=&Lerp_Float(
      A=1
      B=1.5
      Alpha=&FloatFromCurve(
       FloatCurve=DataInterface() {propertyValues {Curve {keys [
           {interpMode=RCIM_Cubic, tangentMode=RCTM_User, time=0, value=0}
           {arriveTangent=0.7625, interpMode=RCIM_Cubic, leaveTangent=0.7625, tangentMode=RCTM_User, time=0.9, value=0.2}
           {arriveTangent=12.0097, interpMode=RCIM_Cubic, leaveTangent=12.0097, tangentMode=RCTM_User, time=1, value=1}
          ]}}}
       CurveIndex=&RandomRangeFloat(Minimum=0, Maximum=1)
      )
     )
    )
    ScaleA=true
    ScaleRGB=false
   )
   &InheritSourceMovement(ApplyForce=false, ApplyPosition=true, ApplyPositionFalloff=false, Source Velocity:Vector3f=$Particles.SkeletalMeshLocation.SampledVelocity) {enabled=false}
   &SampleSkeletalMesh(
    AdditionalYawPitchRoll=false
    Bone Sampling Mode:ENiagaraBoneSamplingMode=::NewEnumerator4
    Skeletal Mesh:NiagaraDataInterfaceSkeletalMesh=$User.Skeletal Mesh
   ) {enabled=false}
   &Drag(Drag=3, Use Linear Drag=true, Use Rotational Drag=true)
   &SolveForcesAndVelocity(Clamp Velocity=false, Limit Acceleration=false)
  ]
  rendererClasses [&NiagaraSpriteRendererProperties]
  renderers [{rendererClass=&NiagaraSpriteRendererProperties}]
 }
}
summary {
 systemName=NS_VampireBossAmbient
 emitters [
  {emitterName=Particle_Smoke, rendererClasses [&NiagaraSpriteRendererProperties]}
  {bEnabled=false, emitterName=Empty, rendererClasses [&NiagaraSpriteRendererProperties]}
 ]
 userVariables [User.Skeletal Mesh {
   type=NiagaraDataInterfaceSkeletalMesh
   defaultValue {
    struct=NiagaraExt_VariableValue_DataInterface
    value {dataInterface=&self:NiagaraDataInterfaceSkeletalMesh_0}
   }
  }]
}
